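# Opaque material for the Stalker phone booth cabin, faking a lit, flickering interior with a ghostly silhouette behind glass.
# View-dependent parallax and virtual-plane UVs sample T_GhostMask for the figure, light source and panning noise.
# Layered sine flicker, offset by actor position and computed per vertex, supplies lamp modulation.
# Emissive blends a red gradient darkened by the ghost with time-of-day Fresnel tint, divides by eye adaptation, then feeds Fortnite's destructible-damage function.
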Material /FNMStalkerKnife/Gameplay/PhoneBooth/VentQuip/VentQuip_PhoneBooth_A/Materials/M_StalkerCabin
graph {
 properties {bAutomaticallySetUsageInEditor=false, bUseMaterialAttributes=true, usedShadingModels=Default Lit}
 nodes [
  Constant() {r=-0.26}
  n1 = CollectionParameter() {
   collection=&FortniteMaterialParameters
   parameterName=SunlightDirectionalVector
  }
  n2 = VertexNormalWS()
  Divide(A=Add(A=DotProduct(A=@n1, B=@n2)))
  n3 = Time()
  n4 = ScalarParameter() {defaultValue=0.015, parameterName=Param_6}
  n5 = ScalarParameter() {defaultValue=1.234, group=A_Emissive_Flicker, parameterName=Emissive Time Speed Layer 1, sortPriority=1}
  n6 = ScalarParameter() {defaultValue=0.76, group=A_Emissive_Flicker, parameterName=Emissive Time Speed Layer 2, sortPriority=2}
  n7 = ScalarParameter() {defaultValue=1.783, group=A_Emissive_Flicker, parameterName=Emissive Time Speed Layer 3, sortPriority=3}
  n8 = ScalarParameter() {defaultValue=1.37, group=A_Emissive_Flicker, parameterName=Emissive Time Speed Layer 4, sortPriority=4}
  n9 = NamedRerouteDeclaration(None=Multiply(
    A=Multiply(A=@n3, B=@n4) {constB=0.178}
    B=MaterialFunctionCall!(1=@n5, 2=@n6, 3=@n7, 4=@n8) {materialFunction=/Game/Global/Base4Layers_2022/MaterialFunctions/MF_UV2LerpScalar_Layers_New}.Result
   )) {name=FlickTime to UV_1, nodeColor=<rgba 0 0.9297 1 1>}
  LinearInterpolate(Alpha=ConstantBiasScale(None=Sine(None=@n9)))
  n10 = NamedRerouteUsage()
  n11 = NamedRerouteUsage()
  n12 = NamedRerouteUsage()
  LinearInterpolate(A=@n10, B=@n11, Alpha=@n12)
  n13 = LinearInterpolate(A=@n11, B=NamedRerouteUsage()) {constAlpha=0.1}
  LinearInterpolate(A=@n10, B=@n13, Alpha=@n12)
  n14 = ScalarParameter() {defaultValue=1.234, group=A_Emissive_Flicker, parameterName=Emissive Time Speed Layer 1, sortPriority=1}
  n15 = ScalarParameter() {defaultValue=0.76, group=A_Emissive_Flicker, parameterName=Emissive Time Speed Layer 2, sortPriority=2}
  n16 = ScalarParameter() {defaultValue=1.783, group=A_Emissive_Flicker, parameterName=Emissive Time Speed Layer 3, sortPriority=3}
  n17 = ScalarParameter() {defaultValue=1.37, group=A_Emissive_Flicker, parameterName=Emissive Time Speed Layer 4, sortPriority=4}
  n18 = NamedRerouteDeclaration(None=Multiply(
    A=Time()
    B=MaterialFunctionCall(1=@n14, 2=@n15, 3=@n16, 4=@n17) {materialFunction=&MF_UV2LerpScalar_Layers_New}.Result
   )) {name=FlickTime to UV, nodeColor=<rgba 0 0.9297 1 1>}
  LinearInterpolate(
   A=ScalarParameter() {defaultValue=0.1, group=A_Emissive_Flicker, parameterName=FlickerMinBrightnessBias, sortPriority=7}
   Alpha=ConstantBiasScale(None=Sine(None=@n18))
  )
  n19 = AppendVector(A=ScalarParameter() {defaultValue=0.32, parameterName=Ghost Pos X}, B=ScalarParameter() {defaultValue=0.16, parameterName=Ghost Pos Y})
  n20 = AppendVector(A=ScalarParameter() {defaultValue=0.7114, parameterName=Ghost Size X}, B=ScalarParameter() {defaultValue=0.8521, parameterName=Ghost Size Y})
  n21 = Multiply(A=Add(A=@n13, B=@n19) {constB=0.1}, B=@n20) {constB=0.7}
  n22 = Divide(A=Time() {period=1}) {constB=7.37}
  n23 = Add(A=@n22) {constB=123}
  n24 = AppendVector(A=@n21, B=Constant())
  LinearInterpolate(
   A=@n21
   B=ComponentMask(None=MaterialFunctionCall(Hue Shift Percentage=@n23, Texture=@n24) {materialFunction=&HueShift}.Result) {g=true, r=true}
  ) {constAlpha=0.05}
  n25 = Multiply(A=Add(A=@n11, B=@n19) {constB=0.1}, B=@n20) {constB=0.7}
  n26 = AppendVector(A=@n25, B=Constant())
  LinearInterpolate(
   A=@n25
   B=ComponentMask(None=MaterialFunctionCall(Hue Shift Percentage=@n22, Texture=@n26) {materialFunction=&HueShift}.Result) {g=true, r=true}
  ) {constAlpha=0.035}
  n27 = Multiply(A=NamedRerouteUsage(), B=ScalarParameter() {defaultValue=0.6101, parameterName=Roughness})
  n28 = Add(A=NamedRerouteUsage(), B=Constant2Vector() {g=1})
  n29 = Constant2Vector() {g=0.5, r=0.5}
  n30 = OneMinus(None=Divide(A=Distance(A=@n28, B=@n29), B=ScalarParameter() {defaultValue=63.7593, parameterName=DistanceFalloff}))
  n31 = ScalarParameter() {defaultValue=107.0971, parameterName=Falloff}
  n32 = Multiply(
   A=LinearInterpolate(
    A=VectorParameter() {parameterName=ColorRed, defaultValue=<rgba 0.0573 0.008 0.004 1>}.RGB
    B=VectorParameter() {parameterName=Color2Red, defaultValue=<rgba 0.8281 0.2061 0.0735 1>}.RGB
    Alpha=Power(Base=@n30, Exp=@n31)
   )
   B=NamedRerouteUsage()
  )
  n33 = TextureSample(UVs=@n21) {samplerSource=SSM_Wrap_WorldGroupSettings, texture=&/T_GhostMask}
  n34 = Multiply(A=@n33.G) {constB=1.2}
  n35 = TextureSample(UVs=@n25) {samplerSource=SSM_Wrap_WorldGroupSettings, texture=&T_GhostMask}
  n36 = Max(A=LinearInterpolate(A=@n34, B=@n35.R, Alpha=@n35.R) {constB=0}, B=@n34)
  n37 = LinearInterpolate(Alpha=NamedRerouteUsage()) {constA=0.25, constB=1.1}
  n38 = Multiply(A=NamedRerouteUsage(), B=NamedRerouteUsage())
  n39 = NamedRerouteUsage()
  n40 = Multiply(
   A=Multiply(A=Multiply(A=@n36, B=@n37) {constB=0.8}, B=ScalarParameter() {defaultValue=1.1, parameterName=Ghost Opacity})
   B=Saturate(None=Multiply(A=@n38, B=@n39))
  )
  n41 = Multiply(A=LinearInterpolate(A=Saturate(None=@n32), B=Constant(), Alpha=Saturate(None=@n40)), B=NamedRerouteUsage())
  n42 = LinearInterpolate(
   A=Multiply(
    A=MaterialFunctionCall() {materialFunction=&Color_Time_of_Day}.Result
    B=ScalarParameter() {defaultValue=0.5, parameterName=Lumen}
   )
   B=VectorParameter() {parameterName=Glass Fresnel Color, defaultValue=<rgba 0.1318 0.194 0.2344 1>}.RGB
  )
  n43 = CameraVectorWS()
  n44 = VertexNormalWS()
  n45 = OneMinus(None=Power(Base=DotProduct(A=@n43, B=@n44), Exp=ScalarParameter() {defaultValue=0.25, parameterName=Fresnel Glass}) {constExponent=0.25})
  n46 = Divide(A=LinearInterpolate(A=@n41, B=Add(A=@n42, B=@n41), Alpha=Saturate(None=@n45)), B=EyeAdaptation().EyeAdaptation)
  n47 = SetMaterialAttributes!(
   MaterialAttributes
   Base Color=Constant()
   Roughness=@n27
   Emissive Color=RayTracingQualitySwitch(Normal=@n46, RayTraced=Multiply(A=@n46) {constB=5})
   Normal
   Specular=Constant() {r=0.2}
  ) {attributeSetTypes [
    #1
    #2
    #3
    #4
    #5
   ]}
  MaterialFunctionCall(Material_In=@n47) {materialFunction=&MF_FN_DestructibleDamage}
  n48 = Multiply(A=Constant3Vector() {constant=<rgba 1 1 0 0>}, B=CameraVectorWS())
  Multiply(A=@n48) {constB=-1}
  n49 = Normalize(VectorInput=@n48)
  n50 = Constant3Vector() {constant=<rgba 0 0 1 0>}
  NamedRerouteDeclaration(None=MaterialFunctionCall(
    Plane Normal Axis=@n49
    Plane X Axis=CrossProduct(A=@n49, B=@n50)
    Plane Center=ObjectPositionWS()
    UV Scale=ScalarParameter() {defaultValue=-125, parameterName=VirutalPlaneScale}
   ) {materialFunction=&VirtualPlaneCoordinates}.UVs) {name=Virtual P UVs, nodeColor=<rgba 0.3125 1 0 1>}
  n51 = Time()
  n52 = Multiply(A=@n51) {constB=0.235}
  n53 = Add(A=Multiply(A=Sine(None=@n52)) {constB=0.5}) {constB=0.5}
  n54 = Saturate(None=Multiply(A=@n53, B=@n53))
  NamedRerouteDeclaration(None=@n54) {name=CloseMask, nodeColor=<rgba 0.4766 0 1 1>}
  n55 = NamedRerouteUsage()
  n56 = Constant()
  n57 = Multiply(A=Constant3Vector() {constant=<rgba 1 1 0 0>}, B=CameraVectorWS())
  n58 = ScalarParameter() {defaultValue=-0.5, parameterName=Depth}
  n59 = DotProduct(A=@n57, B=VertexNormalWS())
  n60 = Saturate(None=Power(Base=@n59))
  n61 = Divide(A=Multiply(A=@n57, B=LinearInterpolate(B=@n58, Alpha=@n60) {constA=-0.2}), B=@n59)
  n62 = ComponentMask(None=Add(A=AppendVector(A=@n55, B=@n56), B=Transform(None=@n61) {transformSourceType=TRANSFORMSOURCE_World, transformType=TRANSFORM_Tangent})) {g=true, r=true}
  NamedRerouteDeclaration(None=@n62) {name=MaxBump Uvs, nodeColor=<rgba 0.4766 0 1 1>}
  n63 = Sine(None=Multiply(A=@n51) {constB=0.7})
  n64 = Constant() {r=0.5}
  n65 = Multiply(A=Multiply(A=@n63, B=Fmod(A=@n63, B=@n64))) {constB=0.4}
  NamedRerouteDeclaration(None=Saturate(None=Add(A=@n54, B=@n65))) {name=Additional flick, nodeColor=<rgba 1 0 0.6562 1>}
  n66 = Multiply(A=NamedRerouteUsage(), B=Multiply(A=Constant2Vector() {g=3, r=1}) {constB=0.75}) {constB=0.5}
  n67 = NamedRerouteUsage()
  n68 = TextureSample(UVs=Add(A=Panner(Coordinate=@n66) {speedY=-0.1}, B=Multiply(A=@n67) {constB=0.5})) {samplerSource=SSM_Wrap_WorldGroupSettings, texture=&T_GhostMask}
  NamedRerouteDeclaration(None=@n68.A) {name=Simple Pan Noise, nodeColor=<rgba 0 1 0.625 1>}
  n69 = NamedRerouteUsage()
  n70 = Add(A=Multiply(A=@n68.A, B=Multiply(A=@n69, B=@n67))) {constB=0.15}
  NamedRerouteDeclaration(None=Saturate(None=Multiply(A=@n70) {constB=4})) {name=Complex Pan Noise, nodeColor=<rgba 0 1 0.625 1>}
  n71 = ScalarParameter() {defaultValue=1, desc=Modulate over Layer speed, group=A_Emissive_Flicker, parameterName=FlickerMinSpeed, sortPriority=5}
  n72 = MaterialFunctionCall(Float3=ActorPositionWS()) {materialFunction=&BreakOutFloat3Components}
  n73 = Divide(A=Add(A=Add(A=@n72.R, B=@n72.R), B=@n72.B)) {constB=50}
  n74 = LinearInterpolate(
   A=Multiply(A=@n71, B=@n18) {constB=0.7}
   B=@n18
   Alpha=Frac(None=NamedRerouteDeclaration(None=@n73) {name=RandActPosVal, nodeColor=<rgba 0 1 0.6484 1>})
  )
  n75 = Sine(None=@n74) {period=0.5651}
  n76 = Sine(None=@n74) {period=0.442}
  n77 = Add(A=Multiply(A=Add(A=@n75, B=@n76), B=ScalarParameter() {defaultValue=0.2824, group=A_Emissive_Flicker, parameterName=FlickerScale, sortPriority=6}))
  n78 = MaterialFunctionCall(Float3=ActorPositionWS()) {materialFunction=&BreakOutFloat3Components}
  n79 = Divide(A=Add(A=Add(A=@n78.R, B=@n78.R), B=@n78.B)) {constB=50}
  n80 = LinearInterpolate(
   A=Multiply(A=ScalarParameter() {defaultValue=2.2194, parameterName=Param_5}, B=@n9) {constB=0.7}
   B=@n9
   Alpha=Frac(None=NamedRerouteDeclaration(None=@n79) {name=RandActPosVal_1, nodeColor=<rgba 0 1 0.6484 1>})
  )
  n81 = Multiply(A=Add(A=Sine(None=@n80) {period=0.5651}, B=Sine(None=@n80) {period=0.442}), B=ScalarParameter() {defaultValue=1, parameterName=Param_4})
  n82 = Add(A=@n77, B=@n81) {constB=0.5}
  NamedRerouteDeclaration(None=VertexInterpolator(VS=Add(A=@n82) {constB=0.3}).PS) {name=FlickerLamp Modulated, nodeColor=<rgba 0 0.8594 1 1>}
  NamedRerouteDeclaration(None=LinearInterpolate(A=@n62, B=NamedRerouteUsage()) {constAlpha=0.25}) {name=bump UVs, nodeColor=<rgba 0.8516 1 0 1>}
  n83 = NamedRerouteUsage()
  n84 = ScalarParameter() {
   desc=Clamp to prevent the emissive multiply from going below this value.
   group=A_Emissive_Flicker
   parameterName=FlickerMinBrightnessValue
   sortPriority=-1
  }
  n85 = Max(A=@n77, B=@n84)
  n86 = ScalarParameter() {
   desc=Clamp to prevent the emissive multiply from going below this value.
   group=A_Emissive_Flicker
   parameterName=FlickerMinBrightnessValue
   sortPriority=-1
  }
  n87 = Max(A=Add(A=@n81), B=@n86)
  n88 = Multiply(
   A=NamedRerouteDeclaration(None=VertexInterpolator(VS=@n85).PS) {name=Flicker, nodeColor=<rgba 0 0.4375 1 1>}
   B=NamedRerouteDeclaration(None=VertexInterpolator(VS=@n87).PS) {name=FlickerLamp Slow, nodeColor=<rgba 0 0.8594 1 1>}
  )
  n89 = NamedRerouteUsage()
  NamedRerouteDeclaration(None=Multiply(A=Multiply(A=@n83, B=@n88), B=LinearInterpolate(Alpha=@n89) {constA=3, constB=5}) {constB=3}) {name=FlickerLamp, nodeColor=<rgba 0 0.8594 1 1>}
  n90 = NamedRerouteUsage()
  n91 = NamedRerouteUsage()
  n92 = ComponentMask(None=@n91) {g=true}
  n93 = LinearInterpolate(A=@n90, B=LinearInterpolate(A=@n90, B=@n91), Alpha=Saturate(None=Add(A=@n92) {constB=0.2})) {constAlpha=0.7}
  n94 = AppendVector(A=ScalarParameter() {defaultValue=0.384, parameterName=Light Pos X}, B=ScalarParameter() {defaultValue=0.232, parameterName=Light Pos Y})
  n95 = ScalarParameter() {defaultValue=0.5705, parameterName=Light Size X}
  n96 = ScalarParameter() {defaultValue=0.6791, parameterName=Light Size Y}
  n97 = TextureSample(UVs=Multiply(A=Add(A=@n93, B=@n94) {constB=0.05}, B=AppendVector(A=@n95, B=@n96)) {constB=0.85}) {texture=&T_GhostMask}
  NamedRerouteDeclaration(None=Multiply(A=@n97.B, B=ScalarParameter() {defaultValue=0.7, parameterName=Light Source Intensity})) {name=Light Source, nodeColor=<rgba 0.6172 0 1 1>}
  NamedRerouteDeclaration(None=@n97.A) {name=Noise Light, nodeColor=<rgba 0 1 0.2969 1>}
  NamedRerouteDeclaration(None=Multiply(A=OneMinus(None=@n35.A), B=@n33.A)) {name=Noise Move, nodeColor=<rgba 1 0 0.0234 1>}
  NamedRerouteUsage()
  OneMinus(None=@n27)
  Saturate(None=@n36)
  TextureCoordinate()
  TextureCoordinate()
 ]
}
